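# Physics asset for the Recoil Rifle skeletal mesh in the Daisy weapon gameplay assault weapon set.
# A single capsule body on the handle bone stands in as the collision shape for the whole rifle.
# It has no constraints, so the rifle simulates as one rigid piece rather than articulated parts.
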PhysicsAsset /DaisyWeaponGameplay/Items/Weapons/Assault/Meshes/Recoil_Rifle_PhysicsAsset
bodies {handle [{shapeType=Capsule, length=38.7434, radius=15.1989, center=<15.1641 0.6255 4.2927>, rotation=<rot 0 89.9997 -86.2279>}]}
constraints []
