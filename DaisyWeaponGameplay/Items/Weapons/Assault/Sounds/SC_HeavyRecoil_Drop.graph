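# SoundCue for the drop sound of the Heavy Recoil assault weapon.
# A Random node picks one of three equally weighted drop waves, and a Modulator adds slight pitch and volume variation.
# The cue is slightly lowered in pitch and volume, uses weapon-drop attenuation and the loot-collect sound class, and references a stop-oldest weapon-drop concurrency set.
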SoundCue /DaisyWeaponGameplay/Items/Weapons/Assault/Sounds/SC_HeavyRecoil_Drop
root=Modulator_0
nodes [
 {id=Modulator_0, children [Random_1], properties {pitchMax=1.1, pitchMin=0.9, volumeMax=1.1, volumeMin=0.9}}
 {id=Random_1, children [WavePlayer_2,WavePlayer_6,WavePlayer_7], properties {weights [1,1,1]}}
 {id=WavePlayer_2, properties {soundWaveAssetPtr=/DaisyWeaponGameplay/Items/Weapons/Assault/Sounds/SoundWaves/HeavyRecoil_Drop_01}}
 {id=WavePlayer_6, properties {soundWaveAssetPtr=/DaisyWeaponGameplay/Items/Weapons/Assault/Sounds/SoundWaves/HeavyRecoil_Drop_02}}
 {id=WavePlayer_7, properties {soundWaveAssetPtr=/DaisyWeaponGameplay/Items/Weapons/Assault/Sounds/SoundWaves/HeavyRecoil_Drop_03}}
]
properties {
 attenuationSettings=&Fort_Weapon_Drop_Att
 maxDistance=3000
 pitchMultiplier=0.9
 soundClassObject=&Fort_Loot_Collect
 volumeMultiplier=0.9
 concurrencyOverrides {resolutionRule=PreventNew}
 concurrencySet [&Weapon_Drop_StopOldest]
}
