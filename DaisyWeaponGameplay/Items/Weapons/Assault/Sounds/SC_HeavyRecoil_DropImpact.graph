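# Sound cue for the impact sound when the heavy recoil assault weapon is dropped.
# A Random node picks one of three equally weighted impact waves, then a Modulator varies pitch and volume slightly on each play.
# It uses weapon-drop attenuation, the loot-collect sound class and a shared weapon-drop concurrency set that limits overlapping drops.
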SoundCue /DaisyWeaponGameplay/Items/Weapons/Assault/Sounds/SC_HeavyRecoil_DropImpact
root=Modulator_0
nodes [
 {id=Modulator_0, children [Random_0], properties {pitchMax=1.1, pitchMin=0.9, volumeMax=1.1, volumeMin=0.9}}
 {id=Random_0, children [WavePlayer_0,WavePlayer_1,WavePlayer_3], properties {weights [1,1,1]}}
 {id=WavePlayer_0, properties {soundWaveAssetPtr=/DaisyWeaponGameplay/Items/Weapons/Assault/Sounds/SoundWaves/HeavyRecoil_DropImpact_01}}
 {id=WavePlayer_1, properties {soundWaveAssetPtr=/DaisyWeaponGameplay/Items/Weapons/Assault/Sounds/SoundWaves/HeavyRecoil_DropImpact_02}}
 {id=WavePlayer_3, properties {soundWaveAssetPtr=/DaisyWeaponGameplay/Items/Weapons/Assault/Sounds/SoundWaves/HeavyRecoil_DropImpact_03}}
]
properties {
 attenuationSettings=&Fort_Weapon_Drop_Att
 maxDistance=3000
 soundClassObject=&Fort_Loot_Collect
 volumeMultiplier=0.8
 concurrencyOverrides {resolutionRule=PreventNew}
 concurrencySet [&Weapon_Drop_StopOldest]
}
