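# Sound cue for the Heavy Recoil assault weapon's gunshot, layering close, medium and distant fire sounds by range.
# Each layer picks a random wave, passes it through a modulator and applies its own attenuation, with random pitch variation on the distant layer.
# The close layer uses a third-person fire sound class, while medium and distant share a distant class through a sub-mixer.
# Local and global gunfire concurrency sets apply, and debug is enabled.
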SoundCue /DaisyWeaponGameplay/Items/Weapons/Assault/Sounds/SC_HeavyRecoil_Fire
root=Mixer_1
nodes [
 {
  id=Attenuation_118
  children [Modulator_4]
  properties {attenuationSettings=&Fort_Weapon_Projectile_Close_Athena_Att}
 }
 {
  id=Attenuation_119
  children [Modulator_2]
  properties {attenuationSettings=&Fort_Weapon_Projectile_Medium_Mono_Athena_Att}
 }
 {
  id=Attenuation_120
  children [Modulator_1]
  properties {attenuationSettings=&Fort_Weapon_Projectile_Distant_Mono_Athena_Att}
 }
 {id=Mixer_0, children [Attenuation_119, Attenuation_120], properties {inputVolume [0.3,0.35]}}
 {id=Mixer_1, children [SoundClass_77, SoundClass_76], properties {inputVolume [0.5,0.6]}}
 {id=Modulator_1, children [Random_3], properties {pitchMax=1.1, pitchMin=0.9, volumeMax=0.8, volumeMin=0.7}}
 {id=Modulator_2, children [Random_2], properties {volumeMax=1.7, volumeMin=1.6}}
 {id=Modulator_4, children [Random_0]}
 {
  id=Random_0
  children [WavePlayer_0,WavePlayer_5,WavePlayer_6,WavePlayer_7,WavePlayer_8,WavePlayer_9]
  properties {bShouldExcludeFromBranchCulling=true, weights [1,1,1,1,1,1]}
 }
 {
  id=Random_2
  children [WavePlayer_13, WavePlayer_14, WavePlayer_15, WavePlayer_18]
  properties {bShouldExcludeFromBranchCulling=true, weights [1,1,1,1]}
 }
 {
  id=Random_3
  children [WavePlayer_19, WavePlayer_21, WavePlayer_22, WavePlayer_23]
  properties {bShouldExcludeFromBranchCulling=true, weights [1,1,1,1]}
 }
 {id=SoundClass_76, children [Mixer_0], properties {soundClassOverride=&Fort_WeaponRanged_Fire_Distant}}
 {
  id=SoundClass_77
  children [Attenuation_118]
  properties {soundClassOverride=&Fort_WeaponRanged_Fire_3P_PSM}
 }
 {id=WavePlayer_0, properties {soundWaveAssetPtr=/DaisyWeaponGameplay/Items/Weapons/Assault/Sounds/SoundWaves/HeavyRecoil_Fire_Close_01}}
 {id=WavePlayer_13, properties {soundWaveAssetPtr=/DaisyWeaponGameplay/Items/Weapons/Assault/Sounds/SoundWaves/HeavyRecoil_Fire_Medium_01}}
 {id=WavePlayer_14, properties {soundWaveAssetPtr=/DaisyWeaponGameplay/Items/Weapons/Assault/Sounds/SoundWaves/HeavyRecoil_Fire_Medium_02}}
 {id=WavePlayer_15, properties {soundWaveAssetPtr=/DaisyWeaponGameplay/Items/Weapons/Assault/Sounds/SoundWaves/HeavyRecoil_Fire_Medium_03}}
 {id=WavePlayer_18, properties {soundWaveAssetPtr=/DaisyWeaponGameplay/Items/Weapons/Assault/Sounds/SoundWaves/HeavyRecoil_Fire_Medium_04}}
 {id=WavePlayer_19, properties {soundWaveAssetPtr=/DaisyWeaponGameplay/Items/Weapons/Assault/Sounds/SoundWaves/HeavyRecoil_Fire_Distant_01}}
 {id=WavePlayer_21, properties {soundWaveAssetPtr=/DaisyWeaponGameplay/Items/Weapons/Assault/Sounds/SoundWaves/HeavyRecoil_Fire_Distant_02}}
 {id=WavePlayer_22, properties {soundWaveAssetPtr=/DaisyWeaponGameplay/Items/Weapons/Assault/Sounds/SoundWaves/HeavyRecoil_Fire_Distant_03}}
 {id=WavePlayer_23, properties {soundWaveAssetPtr=/DaisyWeaponGameplay/Items/Weapons/Assault/Sounds/SoundWaves/HeavyRecoil_Fire_Distant_04}}
 {id=WavePlayer_5, properties {soundWaveAssetPtr=/DaisyWeaponGameplay/Items/Weapons/Assault/Sounds/SoundWaves/HeavyRecoil_Fire_Close_02}}
 {id=WavePlayer_6, properties {soundWaveAssetPtr=/DaisyWeaponGameplay/Items/Weapons/Assault/Sounds/SoundWaves/HeavyRecoil_Fire_Close_03}}
 {id=WavePlayer_7, properties {soundWaveAssetPtr=/DaisyWeaponGameplay/Items/Weapons/Assault/Sounds/SoundWaves/HeavyRecoil_Fire_Close_04}}
 {id=WavePlayer_8, properties {soundWaveAssetPtr=/DaisyWeaponGameplay/Items/Weapons/Assault/Sounds/SoundWaves/HeavyRecoil_Fire_Close_05}}
 {id=WavePlayer_9, properties {soundWaveAssetPtr=/DaisyWeaponGameplay/Items/Weapons/Assault/Sounds/SoundWaves/HeavyRecoil_Fire_Close_06}}
]
properties {
 bDebug=true
 maxDistance=27500
 volumeMultiplier=1
 concurrencySet [&Weapon_Gunfire_Local_Conc, &Weapon_Gunfire_Global_Conc]
}
