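# First-person gunfire SoundCue for the HeavyRecoil assault weapon, layering shot, mechanism and shell impact sounds.
# Random close-shot waves mix with quieter random mechanical waves, which get a slight random delay plus pitch and volume modulation, then pass through a shared modulator.
# A final mixer adds random shell impacts at low level with their own pitch variation.
# It uses first-person weapon attenuation, sound class and gunfire concurrency, and has debug enabled.
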SoundCue /DaisyWeaponGameplay/Items/Weapons/Assault/Sounds/SC_HeavyRecoil_Fire_1P
root=Mixer_5
nodes [
 {id=Delay_1, children [Random_4], properties {delayMax=0.05}}
 {id=Mixer_4, children [Random_2,Modulator_8], properties {inputVolume [1,0.25]}}
 {id=Mixer_5, children [Modulator_2,Modulator_9], properties {inputVolume [1,0.22]}}
 {id=Modulator_2, children [Mixer_4]}
 {id=Modulator_8, children [Delay_1], properties {pitchMax=1.3, pitchMin=1, volumeMax=1, volumeMin=0.8}}
 {id=Modulator_9, children [Random_5], properties {pitchMax=1.1, pitchMin=0.9}}
 {
  id=Random_2
  children [WavePlayer_9, WavePlayer_11, WavePlayer_12, WavePlayer_13, WavePlayer_15, WavePlayer_16]
  properties {bShouldExcludeFromBranchCulling=true, weights [1,1,1,1,1,1]}
 }
 {
  id=Random_4
  children [WavePlayer_17, WavePlayer_23, WavePlayer_27, WavePlayer_28, WavePlayer_29]
  properties {bShouldExcludeFromBranchCulling=true, weights [1,1,1,1,1]}
 }
 {
  id=Random_5
  children [WavePlayer_30, WavePlayer_31, WavePlayer_32, WavePlayer_33, WavePlayer_34, WavePlayer_35]
  properties {bShouldExcludeFromBranchCulling=true, weights [1,1,1,1,1,1]}
 }
 {id=WavePlayer_11, properties {soundWaveAssetPtr=/DaisyWeaponGameplay/Items/Weapons/Assault/Sounds/SoundWaves/HeavyRecoil_Fire_Close_02}}
 {id=WavePlayer_12, properties {soundWaveAssetPtr=/DaisyWeaponGameplay/Items/Weapons/Assault/Sounds/SoundWaves/HeavyRecoil_Fire_Close_03}}
 {id=WavePlayer_13, properties {soundWaveAssetPtr=/DaisyWeaponGameplay/Items/Weapons/Assault/Sounds/SoundWaves/HeavyRecoil_Fire_Close_04}}
 {id=WavePlayer_15, properties {soundWaveAssetPtr=/DaisyWeaponGameplay/Items/Weapons/Assault/Sounds/SoundWaves/HeavyRecoil_Fire_Close_05}}
 {id=WavePlayer_16, properties {soundWaveAssetPtr=/DaisyWeaponGameplay/Items/Weapons/Assault/Sounds/SoundWaves/HeavyRecoil_Fire_Close_06}}
 {id=WavePlayer_17, properties {soundWaveAssetPtr=/DaisyWeaponGameplay/Items/Weapons/Assault/Sounds/SoundWaves/HeavyRecoil_Fire_Close_Mech_01}}
 {id=WavePlayer_23, properties {soundWaveAssetPtr=/DaisyWeaponGameplay/Items/Weapons/Assault/Sounds/SoundWaves/HeavyRecoil_Fire_Close_Mech_02}}
 {id=WavePlayer_27, properties {soundWaveAssetPtr=/DaisyWeaponGameplay/Items/Weapons/Assault/Sounds/SoundWaves/HeavyRecoil_Fire_Close_Mech_03}}
 {id=WavePlayer_28, properties {soundWaveAssetPtr=/DaisyWeaponGameplay/Items/Weapons/Assault/Sounds/SoundWaves/HeavyRecoil_Fire_Close_Mech_04}}
 {id=WavePlayer_29, properties {soundWaveAssetPtr=/DaisyWeaponGameplay/Items/Weapons/Assault/Sounds/SoundWaves/HeavyRecoil_Fire_Close_Mech_05}}
 {id=WavePlayer_30, properties {soundWaveAssetPtr=/DaisyWeaponGameplay/Items/Weapons/Assault/Sounds/SoundWaves/HeavyRecoil_ShellImpact_01}}
 {id=WavePlayer_31, properties {soundWaveAssetPtr=/DaisyWeaponGameplay/Items/Weapons/Assault/Sounds/SoundWaves/HeavyRecoil_ShellImpact_02}}
 {id=WavePlayer_32, properties {soundWaveAssetPtr=/DaisyWeaponGameplay/Items/Weapons/Assault/Sounds/SoundWaves/HeavyRecoil_ShellImpact_03}}
 {id=WavePlayer_33, properties {soundWaveAssetPtr=/DaisyWeaponGameplay/Items/Weapons/Assault/Sounds/SoundWaves/HeavyRecoil_ShellImpact_04}}
 {id=WavePlayer_34, properties {soundWaveAssetPtr=/DaisyWeaponGameplay/Items/Weapons/Assault/Sounds/SoundWaves/HeavyRecoil_ShellImpact_05}}
 {id=WavePlayer_35, properties {soundWaveAssetPtr=/DaisyWeaponGameplay/Items/Weapons/Assault/Sounds/SoundWaves/HeavyRecoil_ShellImpact_06}}
 {id=WavePlayer_9, properties {soundWaveAssetPtr=/DaisyWeaponGameplay/Items/Weapons/Assault/Sounds/SoundWaves/HeavyRecoil_Fire_Close_01}}
]
properties {
 attenuationSettings=&Fort_Weapon_Projectile_FirstPerson_Athena_Att
 bDebug=true
 maxDistance=2097152
 soundClassObject=&Fort_WeaponRanged_Fire_1P_PSM
 volumeMultiplier=1
 concurrencyOverrides {maxCount=1, resolutionRule=PreventNew}
 concurrencySet [&Weapon_Gunfire_FirstPerson_Conc]
}
