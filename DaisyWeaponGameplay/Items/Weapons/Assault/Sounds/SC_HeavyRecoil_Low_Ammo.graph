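# Sound cue for the low-ammo sound of the HeavyRecoil assault weapon.
# A single wave player feeds HeavyRecoil_LowAmmo_02 straight to the output, with no randomization or mixing.
# It uses close-range Athena projectile attenuation, a low-ammo degraded ranged-weapon sound class, a slight volume boost, and a concurrency override that prevents new instances once the limit is reached.
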SoundCue /DaisyWeaponGameplay/Items/Weapons/Assault/Sounds/SC_HeavyRecoil_Low_Ammo
root=WavePlayer_2
nodes [{id=WavePlayer_2, properties {soundWaveAssetPtr=/DaisyWeaponGameplay/Items/Weapons/Assault/Sounds/SoundWaves/HeavyRecoil_LowAmmo_02}}]
properties {
 attenuationSettings=&Fort_Weapon_Projectile_Close_Athena_Att
 bOverrideConcurrency=true
 maxDistance=4950
 soundClassObject=/Game/Sounds/FortSoundClassesAndModes/Fort_WeaponRanged_LowAmmo_Degraded_PSM
 volumeMultiplier=1.1
 concurrencyOverrides {resolutionRule=PreventNew}
}
