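# Sound cue for picking up the Heavy Recoil assault weapon, classed as a loot-collect sound.
# A random node picks one of three equally weighted pickup waves and feeds it through a modulator left at default settings.
# The cue lowers volume and pitch and uses shared weapon-pickup attenuation.
# A concurrency override caps simultaneous instances at two, stopping the oldest.
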SoundCue /DaisyWeaponGameplay/Items/Weapons/Assault/Sounds/SC_HeavyRecoil_Pickup
root=Modulator_0
nodes [
 {id=Modulator_0, children [Random_0]}
 {id=Random_0, children [WavePlayer_1,WavePlayer_2,WavePlayer_3], properties {weights [1,1,1]}}
 {id=WavePlayer_1, properties {soundWaveAssetPtr=/DaisyWeaponGameplay/Items/Weapons/Assault/Sounds/SoundWaves/HeavyRecoil_Pickup_01}}
 {id=WavePlayer_2, properties {soundWaveAssetPtr=/DaisyWeaponGameplay/Items/Weapons/Assault/Sounds/SoundWaves/HeavyRecoil_Pickup_02}}
 {id=WavePlayer_3, properties {soundWaveAssetPtr=/DaisyWeaponGameplay/Items/Weapons/Assault/Sounds/SoundWaves/HeavyRecoil_Pickup_03}}
]
properties {
 attenuationSettings=&Fort_Weapon_Pickup_Att
 bOverrideConcurrency=true
 maxDistance=3500
 pitchMultiplier=0.9
 soundClassObject=&Fort_Loot_Collect
 volumeMultiplier=0.2
 concurrencyOverrides {maxCount=2, resolutionRule=StopOldest}
}
