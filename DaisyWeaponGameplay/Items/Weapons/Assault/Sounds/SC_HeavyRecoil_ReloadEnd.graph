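# Sound cue for the reload-end foley of the Heavy Recoil assault weapon.
# A Random node picks one of four equally weighted wave variants and feeds a Modulator at the root.
# The cue plays in the weapon foley sound class at reduced volume, with shared reload attenuation and a concurrency override.
# Its debug flag is switched on.
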SoundCue /DaisyWeaponGameplay/Items/Weapons/Assault/Sounds/SC_HeavyRecoil_ReloadEnd
root=Modulator_0
nodes [
 {id=Modulator_0, children [Random_2]}
 {id=Random_2, children [WavePlayer_6,WavePlayer_7,WavePlayer_8,WavePlayer_9], properties {weights [1,1,1,1]}}
 {id=WavePlayer_6, properties {soundWaveAssetPtr=/DaisyWeaponGameplay/Items/Weapons/Assault/Sounds/SoundWaves/HeavyRecoil_ReloadEndNew_01}}
 {id=WavePlayer_7, properties {soundWaveAssetPtr=/DaisyWeaponGameplay/Items/Weapons/Assault/Sounds/SoundWaves/HeavyRecoil_ReloadEndNew_02}}
 {id=WavePlayer_8, properties {soundWaveAssetPtr=/DaisyWeaponGameplay/Items/Weapons/Assault/Sounds/SoundWaves/HeavyRecoil_ReloadEndNew_03}}
 {id=WavePlayer_9, properties {soundWaveAssetPtr=/DaisyWeaponGameplay/Items/Weapons/Assault/Sounds/SoundWaves/HeavyRecoil_ReloadEndNew_04}}
]
properties {
 attenuationSettings=&Fort_Weapon_Reload_Att
 bDebug=true
 bOverrideConcurrency=true
 maxDistance=2250
 soundClassObject=&Fort_WeaponFoley
 volumeMultiplier=0.4
}
