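# Sound cue for the magazine-insert foley of the HeavyRecoil assault weapon's reload.
# A Random node picks one of three equally weighted insert wave variations and feeds a root Modulator left at its default settings.
# The cue lowers overall pitch and volume, uses weapon reload attenuation and the weapon foley sound class, overrides concurrency, and has its debug flag enabled.
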SoundCue /DaisyWeaponGameplay/Items/Weapons/Assault/Sounds/SC_HeavyRecoil_ReloadInsert
root=Modulator_0
nodes [
 {id=Modulator_0, children [Random_1]}
 {id=Random_1, children [WavePlayer_0,WavePlayer_1,WavePlayer_4], properties {weights [1,1,1]}}
 {id=WavePlayer_0, properties {soundWaveAssetPtr=/DaisyWeaponGameplay/Items/Weapons/Assault/Sounds/SoundWaves/HeavyRecoil_ReloadInsert_01}}
 {id=WavePlayer_1, properties {soundWaveAssetPtr=/DaisyWeaponGameplay/Items/Weapons/Assault/Sounds/SoundWaves/HeavyRecoil_ReloadInsert_02}}
 {id=WavePlayer_4, properties {soundWaveAssetPtr=/DaisyWeaponGameplay/Items/Weapons/Assault/Sounds/SoundWaves/HeavyRecoil_ReloadInsert_03}}
]
properties {
 attenuationSettings=&Fort_Weapon_Reload_Att
 bDebug=true
 bOverrideConcurrency=true
 maxDistance=2250
 pitchMultiplier=0.9
 soundClassObject=&Fort_WeaponFoley
 volumeMultiplier=0.652
}
